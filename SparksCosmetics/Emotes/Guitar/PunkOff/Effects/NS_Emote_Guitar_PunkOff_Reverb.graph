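# Niagara effect for the PunkOff guitar emote's reverb: a brief purple, cyan and blue burst, likely a visual sound pulse.
# AirBurst meshes and Glows sprites use spawn groups for layered looks, alongside short-lived directional and spike streaks.
# Two BurstRing emitters grow and fade oriented meshes, feeding randomly offset timelines into dynamic material parameters.
# SpikeStreaks alone rescales for front-end preview via user parameters; the Light emitter is empty.
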NiagaraSystem /SparksCosmetics/Emotes/Guitar/PunkOff/Effects/NS_Emote_Guitar_PunkOff_Reverb
dependencies {
 usedDataInterfaces [&NiagaraDataInterfaceCurve, &NiagaraDataInterfaceColorCurve, &NiagaraDataInterfaceVector2DCurve]
 usedDynamicInputs [
  &MaskLinearColorBySpawnGroup
  &Multiply_Vector2D_ByFloat
  &FNTimeOfDayColor
 ]
 usedModules [
  &self:SystemSpawnScript.SystemScriptSource.SystemScriptGraph.NiagaraNodeAssignment_0.SetVariables_#1
  &SystemState
 ]
}
emitters {
 AirBurst {
  emitterUpdateScript [
   &EmitterState()
   &SpawnBurst_Instantaneous(Use Spawn Probability=false, UseLoopCountLimit=false)
   &SpawnBurst_Instantaneous(Spawn Group=1, Use Spawn Probability=false, UseLoopCountLimit=false) {module=SpawnBurst_Instantaneous001}
  ]
  particleSpawnScript [
   /SetVariables(Particles.VisibilityTag=&SpawnGroup()) {isSetParameters=true}
   &InitializeParticle(
    Color=&MergeHelper_ColorSwitch(
     0=<rgba 0.6515 0.1883 1 0.55>
     1=<rgba 1.5656 0.3059 3 2.5>
     2=?
     3=?
     Index=&SpawnGroup()
     Num=::NewEnumerator0
    )
    Lifetime=0.28
    Mesh Scale=&MergeHelper_VectorSwitch(0=<0.1 0.1 0.14>, 1=<0.07 0.04 0.045>, 2=?, 3=?, Index=&SpawnGroup(), Num=::NewEnumerator0)
    Mesh Scale Mode=::NewEnumerator4
    Position Offset=&MergeHelper_VectorSwitch(0=<0 -0.5 0>, 1=<0 -2 0>, 2=?, 3=?, Index=&SpawnGroup(), Num=::NewEnumerator0)
    UsePositionOffset=true
    Write Material Random=true
   )
   &InitialMeshOrientation(
    Mesh Axis To Orient=&MergeHelper_VectorSwitch(0=<0 1 0>, 1=<0 1 0>, 2=?, 3=?, Index=&SpawnGroup(), Num=::NewEnumerator0)
    Mesh Orientation Mode=::NewEnumerator2
    Orientation Vector=&MergeHelper_VectorSwitch(0=<0 0 1>, 1=<1 0 0>, 2=?, 3=?, Index=&SpawnGroup(), Num=::NewEnumerator0)
    Use Rotation Vector=false
   )
  ]
  particleUpdateScript [
   &ParticleState()
   &ScaleMeshSize(Scale Factor=&VectorFromFloat(Value=&FloatFromCurve(
      FloatCurve=DataInterface() {propertyValues {Curve {
         defaultValue=0
         keys [
          {arriveTangent=4.1785, interpMode=RCIM_Cubic, leaveTangent=4.1785, tangentMode=RCTM_User, time=0, value=0}
          {arriveTangent=1.1179, interpMode=RCIM_Cubic, leaveTangent=1.1179, tangentMode=RCTM_User, time=0.2042, value=0.6228}
          {arriveTangent=1.5169, interpMode=RCIM_Cubic, leaveTangent=1.5169, time=0.6975, value=0.8784}
          {arriveTangent=1.7528, interpMode=RCIM_Cubic, leaveTangent=1.7528, tangentMode=RCTM_User, time=1.0018, value=1.8327}
         ]
        }}}
      Scale Curve=2
     ))) {enabled=false}
   &ScaleMeshSize(Scale Factor=&VectorFromFloat(Value=&FloatFromCurve(
      FloatCurve=DataInterface() {propertyValues {Curve {
         defaultValue=0
         keys [
          {arriveTangent=1.7692, interpMode=RCIM_Cubic, leaveTangent=1.7692, tangentMode=RCTM_User, time=0, value=0}
          {arriveTangent=1.1179, interpMode=RCIM_Cubic, leaveTangent=1.1179, tangentMode=RCTM_User, time=0.3121, value=0.4245}
          {arriveTangent=2.7762, interpMode=RCIM_Cubic, leaveTangent=2.7762, tangentMode=RCTM_User, time=0.6848, value=1.5339}
          {arriveTangent=-0.0025, interpMode=RCIM_Cubic, leaveTangent=-0.0025, tangentMode=RCTM_User, time=1.0018, value=1.8327}
         ]
        }}}
      Scale Curve=3
     ))) {module=ScaleMeshSize001}
   &DynamicMaterialParameters(
    Index 0 Param 1=&FloatFromCurve(FloatCurve=DataInterface() {propertyValues {Curve {keys [<at 0 1>,<at 1 0>]}}})
    Index 0 Param 2=&FloatFromCurve(FloatCurve=DataInterface() {propertyValues {Curve {defaultValue=1, keys [{tangentMode=RCTM_User, time=0, value=0}, {tangentMode=RCTM_User, time=1, value=1}]}}})
    Index 0 Param 4=&FloatFromCurve(FloatCurve=DataInterface() {propertyValues {Curve {keys [<at 0 0>,<at 0.03 1>,<at 0.77 1>,<at 1 0>]}}})
    Module.Index0Float0Write=true
    Module.Index0Float1Write=true
    Module.Index0Float2Write=true
    Module.Index0Float3Write=true
   )
   &ScaleColor(
    Scale Alpha=&FloatFromCurve(FloatCurve=DataInterface() {propertyValues {Curve {
        defaultValue=1
        keys [
         {arriveTangent=27.4055, interpMode=RCIM_Cubic, leaveTangent=27.4055, tangentMode=RCTM_User, time=0.0064, value=0.9886}
         {arriveTangent=-0.1862, interpMode=RCIM_Cubic, leaveTangent=-0.1862, time=0.0654, value=1.535}
         {arriveTangent=-1.6424, interpMode=RCIM_Cubic, leaveTangent=-1.6424, time=0.2344, value=0.9462}
         {arriveTangent=-0.3004, interpMode=RCIM_Cubic, leaveTangent=-0.3004, tangentMode=RCTM_User, time=1, value=0}
        ]
       }}})
    ScaleA=true
    ScaleRGB=true
   )
   /SetVariables(Particles.Position=&AddVectorToPosition(
     Position=$Particles.Position
     Vector=&MakeVector(
      X=0
      Y=&FloatFromCurve(
       FloatCurve=DataInterface() {propertyValues {Curve {keys [
           {interpMode=RCIM_Cubic, tangentMode=RCTM_User, time=0, value=1}
           {arriveTangent=-0.6867, interpMode=RCIM_Cubic, leaveTangent=-0.6867, tangentMode=RCTM_User, time=1, value=0}
          ]}}}
       Scale Curve=0.8
      )
      Z=0
     )
    )) {isSetParameters=true}
  ]
  rendererClasses [&NiagaraMeshRendererProperties]
  renderers [
   {rendererClass=&NiagaraMeshRendererProperties, rendererIndex=0}
   {rendererClass=&NiagaraMeshRendererProperties, rendererIndex=1}
  ]
 }
 BurstRing {
  emitterSpawnScript [/SetVariables(Emitter.float:NiagaraFloat=&RandomRangeFloat(Minimum=0, Maximum=500)) {isSetParameters=true, module=SetVariables_#2}]
  emitterUpdateScript [
   &EmitterState(Life Cycle Mode=::NewEnumerator1, Loop Behavior=::NewEnumerator1, Loop Duration=0.3, UseLoopDelay=false)
   &SpawnRate(
    SpawnRate=&FloatFromCurve(
     FloatCurve=DataInterface() {propertyValues {Curve {keys [
         {arriveTangent=63.6346, interpMode=RCIM_Cubic, leaveTangent=63.6346, tangentMode=RCTM_User, time=0, value=30}
         {arriveTangent=-19.1737, interpMode=RCIM_Cubic, leaveTangent=-19.1737, tangentMode=RCTM_User, time=1, value=0}
        ]}}}
     CurveIndex=$Emitter.NormalizedLoopAge
    )
    Use Spawn Probability=false
   ) {enabled=false}
   &SpawnBurst_Instantaneous(Spawn Count=4, Use Spawn Probability=false, UseLoopCountLimit=false)
  ]
  particleSpawnScript [
   &InitializeParticle(
    Color=<rgba 0.125 0.9167 2.5 10>
    Lifetime=&FloatFromCurve(
     FloatCurve=DataInterface() {propertyValues {Curve {keys [<at 0 0.4>,<at 1 0.1>]}}}
     CurveIndex=&ReturnNormalizedExecIndex()
     Scale Curve=1.25
    )
    Mesh Scale Mode=::NewEnumerator3
    Mesh Uniform Scale=&Multiply_Float(
     A=0.2
     B=&FloatFromCurve(
      FloatCurve=DataInterface() {propertyValues {Curve {keys [
          {interpMode=RCIM_Cubic, leaveTangent=0.7118, tangentMode=RCTM_Break, time=0, value=0.4}
          {arriveTangent=0.486, tangentMode=RCTM_User, time=1, value=1}
         ]}}}
      CurveIndex=&ReturnNormalizedExecIndex()
     )
    )
    Sprite Size Mode=::NewEnumerator3
    UsePositionOffset=true
    Write Material Random=true
   )
   &Color(Color=&MultiplyLinearColor(
     Linear Color=$Particles.Color
     Scale Factor=&MakeVector4(
      X=1
      Y=1
      Z=1
      W=&FloatFromCurve(
       FloatCurve=DataInterface() {propertyValues {Curve {keys [
           {interpMode=RCIM_Cubic, leaveTangent=0.9764, tangentMode=RCTM_Break, time=0, value=0.2}
           {arriveTangent=0.4606, interpMode=RCIM_Cubic, tangentMode=RCTM_Break, time=1, value=1}
          ]}}}
       CurveIndex=&ReturnNormalizedExecIndex()
      )
     )
    ))
   &InitialMeshOrientation(Mesh Axis To Orient=<0 0 1>, Mesh Orientation Mode=::NewEnumerator2, Orientation Vector=<0 1 0>, Use Rotation Vector=false)
  ]
  particleUpdateScript [
   &ParticleState()
   &Timeline(Initial Start Time:NiagaraFloat=$Emitter.float, Play Rate=0.7, Timeline End=100000)
   &DynamicMaterialParameters(
    Index 0 Param 1=&MakeFloatFromVector(Vector=$Particles.Scale, Channel=::NewEnumerator0)
    Index 0 Param 2=2
    Index 0 Param 3=-0.3
    Index 0 Param 4=15
    Index 1 Param 1=$System.Timeline.Time
    Index 1 Param 2=3.634
    Index 1 Param 3=0.5
    Index 1 Param 4=$Particles.Timeline.Time
    Index 2 Param 1=4
    Index 2 Param 2=2
    Index 2 Param 3=&FloatFromCurve(
     FloatCurve=DataInterface() {propertyValues {Curve {keys [{interpMode=RCIM_Cubic, leaveTangent=0.3052, tangentMode=RCTM_Break, time=0.35, value=1}, <at 0.5 0>]}}}
     Scale Curve=8
    )
    Module.Index0Param0Write=true
    Module.Index0Param1Write=true
    Module.Index0Param2Write=true
    Module.Index0Param3Write=true
    Module.Index1Param0Write=true
    Module.Index1Param1Write=true
    Module.Index1Param2Write=true
    Module.Index1Param3Write=true
    Module.Index2Param0Write=true
    Module.Index2Param1Write=true
    Module.Index2Param2Write=true
    Module.Index2Param3Write=true
    Write Parameter Index 1=true
    Write Parameter Index 2=true
   )
   &ScaleMeshSize(Scale Factor=&VectorFromFloat(Value=&FloatFromCurve(
      FloatCurve=DataInterface() {propertyValues {Curve {keys [
          {interpMode=RCIM_Cubic, leaveTangent=1.45, tangentMode=RCTM_Break, time=0, value=0.2}
          {arriveTangent=-0.7179, interpMode=RCIM_Cubic, tangentMode=RCTM_Break, time=0.9999, value=3}
         ]}}}
      Scale Curve=&FloatFromCurve(
       FloatCurve=DataInterface() {propertyValues {Curve {defaultValue=0.2, keys [{tangentMode=RCTM_User, time=0, value=1.5}, {tangentMode=RCTM_User, time=1, value=0.5}]}}}
       CurveIndex=&ReturnNormalizedExecIndex()
      )
     )))
   &ScaleColor(
    Scale Alpha=&FloatFromCurve(
     FloatCurve=DataInterface() {propertyValues {Curve {keys [
         {interpMode=RCIM_Cubic, tangentMode=RCTM_User, time=0, value=0}
         {arriveTangent=-3.9775, interpMode=RCIM_Cubic, leaveTangent=-3.9775, tangentMode=RCTM_User, time=0.2, value=1}
         {tangentMode=RCTM_User, time=0.987, value=0}
        ]}}}
     Scale Curve=0.68
    )
    Scale RGB=&VectorFromFloat(Value=&FloatFromCurve(FloatCurve=DataInterface() {propertyValues {Curve {keys [{tangentMode=RCTM_User, time=0, value=1}, {tangentMode=RCTM_User, time=1, value=1}]}}}))
    ScaleA=true
    ScaleRGB=true
   )
   &CameraOffset(Camera Offset Amount=20)
   /SetVariables(Particles.Position=&AddVectorToPosition(
     Position=$Particles.Position
     Vector=&MakeVector(
      X=0
      Y=&FloatFromCurve(
       FloatCurve=DataInterface() {propertyValues {Curve {keys [
           {interpMode=RCIM_Cubic, tangentMode=RCTM_User, time=0, value=1}
           {arriveTangent=-1.8943, interpMode=RCIM_Cubic, leaveTangent=-1.8943, tangentMode=RCTM_User, time=1, value=0}
          ]}}}
       Scale Curve=3
      )
      Z=0
     )
    )) {isSetParameters=true, module=SetVariables_#3}
  ]
  rendererClasses [&NiagaraMeshRendererProperties]
  renderers [{rendererClass=&NiagaraMeshRendererProperties, rendererIndex=0}]
 }
 BurstRing001 {
  emitterSpawnScript [/SetVariables(Emitter.float:NiagaraFloat=&RandomRangeFloat(Minimum=0, Maximum=500)) {isSetParameters=true, module=SetVariables_#2}]
  emitterUpdateScript [
   &EmitterState(Life Cycle Mode=::NewEnumerator1, Loop Behavior=::NewEnumerator1, Loop Duration=0.32, UseLoopDelay=false)
   &SpawnRate(
    SpawnRate=&FloatFromCurve(
     FloatCurve=DataInterface() {propertyValues {Curve {keys [
         {arriveTangent=63.6346, interpMode=RCIM_Cubic, leaveTangent=63.6346, tangentMode=RCTM_User, time=0, value=30}
         {arriveTangent=-19.1737, interpMode=RCIM_Cubic, leaveTangent=-19.1737, tangentMode=RCTM_User, time=1, value=0}
        ]}}}
     CurveIndex=$Emitter.NormalizedLoopAge
    )
    Use Spawn Probability=false
   )
  ]
  particleSpawnScript [
   &InitializeParticle(
    Color=<rgba 0.0466 0.6163 1 7>
    Lifetime=&FloatFromCurve(
     FloatCurve=DataInterface() {propertyValues {Curve {keys [<at 0 0.4>,<at 1 0.1>]}}}
     CurveIndex=&ReturnNormalizedExecIndex()
     Scale Curve=1.25
    )
    Mesh Scale Mode=::NewEnumerator3
    Mesh Uniform Scale=&Multiply_Float(
     A=0.2
     B=&FloatFromCurve(
      FloatCurve=DataInterface() {propertyValues {Curve {keys [
          {interpMode=RCIM_Cubic, leaveTangent=0.7118, tangentMode=RCTM_Break, time=0, value=0.4}
          {arriveTangent=0.486, tangentMode=RCTM_User, time=1, value=1}
         ]}}}
      CurveIndex=&ReturnNormalizedExecIndex()
     )
    )
    Sprite Size Mode=::NewEnumerator3
    UsePositionOffset=true
    Write Material Random=true
   )
   &Color(Color=&MultiplyLinearColor(
     Linear Color=$Particles.Color
     Scale Factor=&MakeVector4(
      X=1
      Y=1
      Z=1
      W=&FloatFromCurve(
       FloatCurve=DataInterface() {propertyValues {Curve {keys [
           {interpMode=RCIM_Cubic, leaveTangent=0.9764, tangentMode=RCTM_Break, time=0, value=0.2}
           {arriveTangent=0.4606, interpMode=RCIM_Cubic, tangentMode=RCTM_Break, time=1, value=1}
          ]}}}
       CurveIndex=&ReturnNormalizedExecIndex()
      )
     )
    ))
   &InitialMeshOrientation(Mesh Axis To Orient=<0 0 1>, Mesh Orientation Mode=::NewEnumerator2, Orientation Vector=<0 1 0>, Use Rotation Vector=false)
  ]
  particleUpdateScript [
   &ParticleState()
   &Timeline(Initial Start Time:NiagaraFloat=$Emitter.float, Play Rate=0.7, Timeline End=100000)
   &DynamicMaterialParameters(
    Index 0 Param 1=&MakeFloatFromVector(Vector=$Particles.Scale, Channel=::NewEnumerator0)
    Index 0 Param 2=1
    Index 0 Param 3=-20
    Index 0 Param 4=15
    Index 1 Param 1=$System.Timeline.Time
    Index 1 Param 2=3.634
    Index 1 Param 3=0.5
    Index 1 Param 4=$Particles.Timeline.Time
    Index 2 Param 3=&FloatFromCurve(
     FloatCurve=DataInterface() {propertyValues {Curve {keys [{interpMode=RCIM_Cubic, leaveTangent=0.3052, tangentMode=RCTM_Break, time=0.35, value=1}, <at 0.5 0>]}}}
     Scale Curve=8
    )
    Module.Index0Param0Write=true
    Module.Index0Param1Write=true
    Module.Index0Param2Write=true
    Module.Index0Param3Write=true
    Module.Index1Param0Write=true
    Module.Index1Param1Write=true
    Module.Index1Param2Write=true
    Module.Index1Param3Write=true
    Module.Index2Param0Write=true
    Module.Index2Param1Write=true
    Module.Index2Param2Write=true
    Module.Index2Param3Write=true
    Write Parameter Index 1=true
    Write Parameter Index 2=true
   )
   &ScaleMeshSize(Scale Factor=&VectorFromFloat(Value=&FloatFromCurve(
      FloatCurve=DataInterface() {propertyValues {Curve {keys [
          {interpMode=RCIM_Cubic, leaveTangent=1.45, tangentMode=RCTM_Break, time=0, value=0.2}
          {arriveTangent=-0.7179, interpMode=RCIM_Cubic, tangentMode=RCTM_Break, time=0.9999, value=3}
         ]}}}
      Scale Curve=&FloatFromCurve(
       FloatCurve=DataInterface() {propertyValues {Curve {
          defaultValue=0
          keys [
           {interpMode=RCIM_Cubic, tangentMode=RCTM_User, time=0.0138, value=1.4368}
           {arriveTangent=-0.0194, interpMode=RCIM_Cubic, leaveTangent=-0.0194, time=0.3072, value=2.015}
           {interpMode=RCIM_Cubic, tangentMode=RCTM_User, time=1.0066, value=1.4176}
          ]
         }}}
       CurveIndex=&ReturnNormalizedExecIndex()
       Scale Curve=1.55
      )
     )))
   &ScaleColor(
    Scale Alpha=&FloatFromCurve(
     FloatCurve=DataInterface() {propertyValues {Curve {keys [
         {interpMode=RCIM_Cubic, tangentMode=RCTM_User, time=0, value=0}
         {arriveTangent=-3.9775, interpMode=RCIM_Cubic, leaveTangent=-3.9775, tangentMode=RCTM_User, time=0.2, value=1}
         {tangentMode=RCTM_User, time=1, value=0}
        ]}}}
     Scale Curve=0.68
    )
    Scale RGB=&VectorFromFloat(Value=&FloatFromCurve(FloatCurve=DataInterface() {propertyValues {Curve {keys [{tangentMode=RCTM_User, time=0, value=1}, {tangentMode=RCTM_User, time=1, value=1}]}}}))
    ScaleA=true
    ScaleRGB=true
   )
   &CameraOffset(Camera Offset Amount=20)
   /SetVariables(Particles.Position=&AddVectorToPosition(
     Position=$Particles.Position
     Vector=&MakeVector(
      X=0
      Y=&FloatFromCurve(
       FloatCurve=DataInterface() {propertyValues {Curve {keys [
           {interpMode=RCIM_Cubic, tangentMode=RCTM_User, time=0, value=1}
           {arriveTangent=-1.8943, interpMode=RCIM_Cubic, leaveTangent=-1.8943, tangentMode=RCTM_User, time=1, value=0}
          ]}}}
       Scale Curve=2.5
      )
      Z=0
     )
    )) {isSetParameters=true, module=SetVariables_#3}
  ]
  rendererClasses [&NiagaraMeshRendererProperties]
  renderers [{rendererClass=&NiagaraMeshRendererProperties, rendererIndex=0}]
 }
 Directional_Streaks {
  emitterUpdateScript [
   &EmitterState()
   &SpawnBurst_Instantaneous(Spawn Count=3, Use Spawn Probability=true, UseLoopCountLimit=false)
  ]
  particleSpawnScript [
   &InitializeParticle(
    Color Mode=::NewEnumerator0
    Lifetime Max=0.12
    Lifetime Min=0.08
    Lifetime Mode=::NewEnumerator1
    Position Offset=<10 0 0>
    Sprite Size Max=<17 42>
    Sprite Size Min=<10 30>
    Sprite Size Mode=::NewEnumerator2
    Sprite UV Mode=::NewEnumerator1
    UsePositionOffset=true
    Write Material Random=true
   ) {module=InitializeParticle001}
   &AddVelocity(Velocity=&RandomRangeVector(Minimum=<-100 -100 -100>, Maximum=<100 100 100>)) {module=AddVelocity001}
   &ShapeLocation(MaskSpawnGroup=false, Override Seed=false, Sphere Radius=5, Sphere Surface Distribution=1, UseOwnerScale=true) {enabled=false}
   &SubUVAnimation(End Frame=3, SubUV Animation Mode:ENiagara_SubUVLookupMode=::NewEnumerator2)
   &CameraOffset(Camera Offset Amount=25)
  ]
  particleUpdateScript [
   &ParticleState()
   &Drag(
    Drag=&FloatFromCurve(FloatCurve=DataInterface() {propertyValues {Curve {
        defaultValue=1
        keys [{interpMode=RCIM_Cubic, tangentMode=RCTM_User, time=0, value=0}, {interpMode=RCIM_Cubic, tangentMode=RCTM_User, time=1, value=1.5}]
       }}})
    Use Linear Drag=true
    Use Rotational Drag=true
   )
   &SolveForcesAndVelocity(Clamp Velocity=false, Limit Acceleration=false)
   &ScaleSpriteSize(Scale Factor=&Vector2DFromCurve(
     Vector2Curve=DataInterface() {propertyValues {XCurve {keys [<at 0 0.778>,<at 1 1>]}, YCurve {keys [<at 0.0019 0.377>,<at 0.7271 1.4884>,<at 0.9991 1.0026>]}}}
     ScaleCurve=<0.6 1>
    ))
   &ScaleColor(
    Linear Color Curve=DataInterface() {propertyValues {
      AlphaCurve {keys [<at 0 0.5>,<at 1 0>]}
      BlueCurve {keys [<at 0.0279 1>,<at 0.5529 2>,<at 0.6552 1>,<at 0.9857 1>]}
      GreenCurve {keys [<at 0.0279 0.7993>,<at 0.5529 0.2777>,<at 0.6552 0>,<at 0.9857 0>]}
      RedCurve {keys [<at 0.0279 0.045>,<at 0.5529 0>,<at 0.6552 0.2893>,<at 0.9857 0.6442>]}
     }}
    Scale Mode=::NewEnumerator2
   )
  ]
  rendererClasses [&NiagaraSpriteRendererProperties]
  renderers [{rendererClass=&NiagaraSpriteRendererProperties}]
 }
 Glows {
  emitterUpdateScript [
   &EmitterState()
   &SpawnBurst_Instantaneous(Use Spawn Probability=false, UseLoopCountLimit=false)
   &SpawnBurst_Instantaneous(Spawn Group=1, Use Spawn Probability=false, UseLoopCountLimit=false) {module=SpawnBurst_Instantaneous001}
   &SpawnBurst_Instantaneous(Spawn Group=2, Use Spawn Probability=false, UseLoopCountLimit=false) {module=SpawnBurst_Instantaneous002}
  ]
  particleSpawnScript [
   /SetVariables(Particles.VisibilityTag=&SpawnGroup()) {isSetParameters=true}
   &InitializeParticle(
    Color=&MergeHelper_ColorSwitch(
     0=<rgba 0.1107 0 0.526 0.78>
     1=<rgba 0.225 1.9111 3 0.487>
     2=?
     3=?
     Index=&SpawnGroup()
     Num=::NewEnumerator0
    )
    Lifetime=&MergeHelper_FloatSwitch(0=0.15, 1=0.12, 2=0.1, 3=?, Index=&SpawnGroup(), Num=::NewEnumerator1)
    Position Offset=<12 0 0>
    Sprite Rotation Angle=20
    Sprite Rotation Mode=::NewEnumerator0
    Sprite Size=&MergeHelper_Vector2Switch(0=<132 102>, 1=<75 60>, 2=<69 69>, 3=?, 4=?, Index=&SpawnGroup(), Num=::NewEnumerator1)
    Sprite Size Mode=::NewEnumerator4
    UsePositionOffset=true
    Write Material Random=true
   )
   &CameraOffset(Camera Offset Amount=-2)
  ]
  particleUpdateScript [
   &ParticleState()
   &ScaleSpriteSize(
    Scale Sprite Size Mode=::NewEnumerator0
    Uniform Scale Factor=&MergeHelper_FloatSwitch(
     0=&FloatFromCurve(FloatCurve=DataInterface() {propertyValues {Curve {keys [
          {arriveTangent=19.0835, interpMode=RCIM_Cubic, leaveTangent=19.0835, tangentMode=RCTM_User, time=0, value=0.1}
          {arriveTangent=-0.3437, interpMode=RCIM_Cubic, leaveTangent=-0.3437, tangentMode=RCTM_User, time=0.1356, value=1.0174}
          {arriveTangent=-0.5237, interpMode=RCIM_Cubic, leaveTangent=-0.5237, tangentMode=RCTM_User, time=0.7853, value=0.75}
          {arriveTangent=-4.9378, interpMode=RCIM_Cubic, leaveTangent=-4.9378, tangentMode=RCTM_User, time=1, value=0}
         ]}}})
     1=&FloatFromCurve(FloatCurve=DataInterface() {propertyValues {Curve {
         defaultValue=1
         keys [
          {arriveTangent=29.7346, interpMode=RCIM_Cubic, leaveTangent=29.7346, tangentMode=RCTM_User, time=-0.0037, value=0}
          {interpMode=RCIM_Cubic, tangentMode=RCTM_User, time=0.0809, value=1}
          {arriveTangent=-2.0778, interpMode=RCIM_Cubic, leaveTangent=-2.0778, tangentMode=RCTM_User, time=1, value=0}
         ]
        }}})
     2=&FloatFromCurve(FloatCurve=DataInterface() {propertyValues {Curve {
         defaultValue=0
         keys [
          {arriveTangent=76.0132, interpMode=RCIM_Cubic, leaveTangent=76.0132, tangentMode=RCTM_User, time=0, value=0}
          {arriveTangent=-3.1661, interpMode=RCIM_Cubic, leaveTangent=-3.1661, tangentMode=RCTM_User, time=0.056, value=1.0863}
          {interpMode=RCIM_Cubic, tangentMode=RCTM_User, time=1.0112, value=0.0863}
         ]
        }}})
     3=0
     4=0
     Index=0
     Num=::NewEnumerator3
    )
   )
   &ScaleColor(
    Scale Alpha=&FloatFromCurve(FloatCurve=DataInterface() {propertyValues {Curve {keys [<at 0 1>,<at 1 0>]}}})
    ScaleA=true
    ScaleRGB=true
   )
   /SetVariables(Particles.Flare Color:LinearColor=&ColorFromCurve(ColorCurve=DataInterface() {propertyValues {
       AlphaCurve {keys [<at 0 1 RCIM_Cubic>,<at 1 0 RCIM_Cubic>]}
       BlueCurve {keys [
         <at 0.2225 30>
         <at 0.5333 20>
         {arriveTangent=-47.4497, interpMode=RCIM_Cubic, leaveTangent=-47.4497, time=0.78, value=5}
         <at 0.8927 2.947>
         <at 0.9627 1>
         {interpMode=RCIM_Cubic, leaveTangent=-8.2439, time=1, value=1}
        ]}
       GreenCurve {keys [
         <at 0.2225 27.9879>
         <at 0.5333 18.6586>
         {arriveTangent=-50.7824, interpMode=RCIM_Cubic, leaveTangent=-50.7824, time=0.78, value=4.8043}
         <at 0.8927 0.4078>
         <at 0.9627 0.0134>
         {interpMode=RCIM_Cubic, leaveTangent=-9.8739, time=1, value=0.0084}
        ]}
       RedCurve {keys [
         <at 0.2225 15.1508>
         <at 0.5333 10.1005>
         {arriveTangent=-28.0867, interpMode=RCIM_Cubic, leaveTangent=-28.0867, time=0.78, value=0}
         <at 0.8927 0.0064>
         <at 0.9627 0.3339>
         {interpMode=RCIM_Cubic, leaveTangent=0.6881, time=1, value=0.1421}
        ]}
      }})) {isSetParameters=true}
  ]
  rendererClasses [&NiagaraSpriteRendererProperties]
  renderers [
   {rendererClass=&NiagaraSpriteRendererProperties}
   {rendererClass=&NiagaraSpriteRendererProperties, rendererIndex=1}
   {rendererClass=&NiagaraSpriteRendererProperties, rendererIndex=2}
  ]
 }
 Light {rendererClasses [], renderers []}
 SpikeStreaksParticle {
  emitterUpdateScript [
   &EmitterState(
    Enable Distance Culling=true
    Max Distance Response=::KillImmediately
    MaxDistance=3500
    Module.bUseMaxDistance=true
    Module.bUseMinDistance=false
    Scalability Mode=::NewEnumerator1
   )
   &SpawnBurst_Instantaneous(Spawn Count=4, Use Spawn Probability=false)
  ]
  particleSpawnScript [
   /Niagara/Modules/Spawn/Initialization/InitializeParticle(
    Lifetime=0.2
    Mesh Scale=<0.8 0.8 0.8>
    Module.Write Color=true
    Module.Write Lifetime=true
    Module.Write Mass=true
    Module.Write Material Random=true
    Module.Write Position=true
    Module.Write Scale=true
    Module.Write SpriteRotation=false
    Module.Write SpriteSize=true
    Position=&Add_Vector(A=/Niagara/DynamicInputs/Helpers/SimulationPosition(), B=<0 0 0>)
    Sprite Size=&UniformRangedVector2D(Minimum=<40 60>, Maximum=<15 40>)
   )
   &AddVelocityFromPoint(
    Module.Use Falloff Distance=false
    Origin Offset=&RandomRangeVector(Minimum=<-12 0 -1>, Maximum=<12 0 1>)
    Velocity Strength=120
   )
   &CameraOffset(Camera Offset Amount=15)
  ]
  particleUpdateScript [
   &ParticleState()
   &SolveForcesAndVelocity(Clamp Velocity=false, Limit Acceleration=false)
   &ScaleSpriteSize(Scale Factor=&Vector2DFromFloat(Value=&FloatFromCurve(FloatCurve=DataInterface() {propertyValues {Curve {keys [<at 0 0.6>,<at 1 2>]}}}, Scale Curve=0.5)))
   &ScaleColor(
    Scale Alpha=&FloatFromCurve(FloatCurve=DataInterface() {propertyValues {Curve {keys [
         {interpMode=RCIM_Cubic, tangentMode=RCTM_User, time=0, value=1}
         {arriveTangent=-1.4883, interpMode=RCIM_Cubic, leaveTangent=-1.4883, tangentMode=RCTM_User, time=0.3463, value=0.4081}
         {interpMode=RCIM_Cubic, tangentMode=RCTM_User, time=1, value=0}
        ]}}})
    Scale RGB=&Multiply_VectorByFloat(
     Vector=&MakeVectorFromLinearColorRGB(Color=&HueShiftLinearColor(
       Hue=0.05
       Saturation=1
       Value=1.5
       Linear Color=&ColorFromCurve(ColorCurve=DataInterface() {propertyValues {
          AlphaCurve {keys []}
          BlueCurve {keys [
            {arriveTangent=-2.74, interpMode=RCIM_Cubic, time=0, value=10}
            <at 0.0239 5>
            <at 0.0488 1.6219>
            <at 0.4638 0.6263>
            <at 0.5728 0.3647>
            <at 0.6879 0.3 RCIM_Cubic>
           ]}
          GreenCurve {keys [
            {arriveTangent=-4.3862, interpMode=RCIM_Cubic, time=0, value=1}
            <at 0.0239 2.4185>
            <at 0.0488 0.1585>
            <at 0.4638 0.0494>
            <at 0.5728 0.0204>
            <at 0.6879 0.0281 RCIM_Cubic>
           ]}
          RedCurve {keys [
            {arriveTangent=-2.9918, interpMode=RCIM_Cubic, time=0, value=1.75}
            <at 0.0239 0.1136>
            <at 0.0488 0.2758>
            <at 0.4638 0.136>
            <at 0.5728 0.0292>
            <at 0.6879 0.1414 RCIM_Cubic>
           ]}
         }})
      ))
     Float=1
    )
    ScaleA=true
    ScaleRGB=true
   )
   &ScaleSpriteSize(
    Scale Sprite Size Mode=::NewEnumerator0
    Uniform Scale Factor=&MakeCustomFloatFromBool(Bool=$User.bIsFrontEndPreview, True Float=$User.FEScale, False Float=1)
   ) {module=ScaleSpriteSize001}
  ]
  rendererClasses [&NiagaraSpriteRendererProperties]
  renderers [{rendererClass=&NiagaraSpriteRendererProperties}]
 }
}
summary {
 systemName=NS_Emote_Guitar_PunkOff_Reverb
 emitters [
  {emitterName=AirBurst, rendererClasses [&NiagaraMeshRendererProperties]}
  {emitterName=Glows, rendererClasses [&NiagaraSpriteRendererProperties]}
  {emitterName=SpikeStreaksParticle, rendererClasses [&NiagaraSpriteRendererProperties]}
  {emitterName=BurstRing, rendererClasses [&NiagaraMeshRendererProperties]}
  {emitterName=Directional_Streaks, rendererClasses [&NiagaraSpriteRendererProperties]}
  {emitterName=BurstRing001, rendererClasses [&NiagaraMeshRendererProperties]}
  {emitterName=Light, rendererClasses []}
 ]
 userVariables [
  User.bIsFrontEnd {type=NiagaraBool}
  User.bIsFrontEndPreview {type=NiagaraBool}
  User.FEOffset {type=NiagaraFloat, defaultValue=10, description=""}
  User.FEScale {type=NiagaraFloat, defaultValue=0, description=""}
  User.Hitted {type=NiagaraFloat, defaultValue=0, description=""}
 ]
}
